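# Animation Blueprint for the Assemble L backpack cosmetic, animating its skeleton with idle motion.
# The AnimGraph feeds a single Random Sequence Player straight into the output pose.
# It picks between two equally weighted idle sequences and a less frequent battery change animation, each with a short linear blend.
# The event graph holds only unconnected default nodes, so no gameplay state drives the animation.
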AnimBlueprint /BRCosmetics/Accessories/FORT_Backpacks/Backpack_M_MED_Assemble_L/Meshes/M_MED_Assemble_L_Pack_AnimBP
target_skeleton=&/M_MED_Assemble_L_Pack_Skeleton
graphs [
 AnimGraph {nodes [
   n1 = Random Sequence Player!() -> (Pose:Pose Link Structure) {
    is=title AnimGraphNode_RandomPlayer
    entries [
     {
      chanceToPlay=1
      maxLoopCount=0
      maxPlayRate=1
      minLoopCount=0
      minPlayRate=1
      sequence=/BRCosmetics/Accessories/FORT_Backpacks/Backpack_M_MED_Assemble_L/Animations/Assemble_L_Idle
      blendIn {blendOption=Linear, blendTime=0.2, customCurve=null}
     }
     {
      chanceToPlay=1
      maxLoopCount=0
      maxPlayRate=1
      minLoopCount=0
      minPlayRate=1
      sequence=/BRCosmetics/Accessories/FORT_Backpacks/Backpack_M_MED_Assemble_L/Animations/Assemble_L_Idle_02
      blendIn {blendOption=Linear, blendTime=0.2, customCurve=null}
     }
     {
      chanceToPlay=0.2
      maxLoopCount=0
      maxPlayRate=1
      minLoopCount=0
      minPlayRate=1
      sequence=/BRCosmetics/Accessories/FORT_Backpacks/Backpack_M_MED_Assemble_L/Animations/Assemble_L_BatteryChange
      blendIn {blendOption=Linear, blendTime=0.2, customCurve=null}
     }
    ]
   }
   Output Pose(Result=@n1) {layerGroup=None}
  ]}
 EventGraph {nodes [
   TryGetPawnOwner()
   Event BlueprintUpdateAnimation()
  ]}
]
variables []
